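# HUD widget for the Playground memory thermometer that shows a signed preview of a memory change in a small pill with an optional warning icon.
# Set previewText hides the pill at zero, and updates sign and number only when either differs from the cached state.
# Outside mobile, updates play a slide-in and an endlessly looping ping-pong opacity pulse.
# Helper functions tint text or icon red/white, toggle icon visibility, and play a scale-and-fade text animation.
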WidgetBlueprint /FortniteCoreUI/HUD/LTM/Playground/Thermometer/PreviewBar_Memory
animations [
 previewValue {
  binding_count=2
  length_seconds=0.5017
  object=&self:previewValue
  sequence {bindings [
    previewText {
     id=#1
     possessed_class=CommonTextBlock
     tracks [{
       type=MovieSceneFloatTrack
       display_name=Render Opacity
       sections [{
         type=MovieSceneFloatSection
         end_frame=10
         start_frame=0
         channels [None {type=MovieSceneScriptingFloatChannel, keys [[0,0.25],[7,1],[10,1]]}]
        }]
      }]
    }
    Symbol {
     id=#2
     possessed_class=CommonTextBlock
     tracks [{
       type=MovieSceneFloatTrack
       display_name=Render Opacity
       sections [{
         type=MovieSceneFloatSection
         end_frame=10
         start_frame=0
         channels [None {type=MovieSceneScriptingFloatChannel, keys [[0,0.25],[7,1],[10,1]]}]
        }]
      }]
    }
   ]}
 }
 IN {
  binding_count=1
  length_seconds=0.25
  object=&self:IN
  sequence {bindings [PreviewBox {
     id=#3
     possessed_class=SizeBox
     tracks [
      {
       type=MovieSceneFloatTrack
       display_name=Render Opacity
       sections [{type=MovieSceneFloatSection, end_frame=20, start_frame=0, channels [None {type=MovieSceneScriptingFloatChannel, keys [[0,0],[5,1]]}]}]
      }
      {
       type=MovieScene2DTransformTrack
       sections [{
         type=MovieScene2DTransformSection
         end_frame=20
         start_frame=0
         channels [
          Translation.X {type=MovieSceneScriptingFloatChannel, keys [[0,82.8398],[5,0]]}
          Translation.Y {type=MovieSceneScriptingFloatChannel, const [0,1]}
          Angle {type=MovieSceneScriptingFloatChannel, const [0,1]}
          Scale.X {type=MovieSceneScriptingFloatChannel, const [1,1]}
          Scale.Y {type=MovieSceneScriptingFloatChannel, const [1,1]}
          Shear.X {type=MovieSceneScriptingFloatChannel, const [0,1]}
          Shear.Y {type=MovieSceneScriptingFloatChannel, const [0,1]}
         ]
        }]
      }
     ]
    }]}
 }
 Text {
  binding_count=1
  length_seconds=1.05
  object=&self:Text
  sequence {bindings [HorizontalBox_0 {
     id=#4
     possessed_class=HorizontalBox
     tracks [
      {
       type=MovieScene2DTransformTrack
       sections [{
         type=MovieScene2DTransformSection
         end_frame=21
         start_frame=0
         channels [
          Translation.X {type=MovieSceneScriptingFloatChannel, const [0,1]}
          Translation.Y {type=MovieSceneScriptingFloatChannel, const [0,1]}
          Angle {type=MovieSceneScriptingFloatChannel, const [0,1]}
          Scale.X {type=MovieSceneScriptingFloatChannel, keys [[0,1],[20,2],[21,1]]}
          Scale.Y {type=MovieSceneScriptingFloatChannel, keys [[0,1],[20,2],[21,1]]}
          Shear.X {type=MovieSceneScriptingFloatChannel, const [0,1]}
          Shear.Y {type=MovieSceneScriptingFloatChannel, const [0,1]}
         ]
        }]
      }
      {
       type=MovieSceneFloatTrack
       display_name=Render Opacity
       sections [{type=MovieSceneFloatSection, end_frame=21, start_frame=0, channels [None {type=MovieSceneScriptingFloatChannel, keys [[0,1],[20,0],[21,1]]}]}]
      }
     ]
    }]}
 }
]
graphs [
 EventGraph {nodes [Event PreConstruct()]}
 Set previewText {nodes [
   n1 = Set previewText!() -> (then, PreviewTextNUM:double=~, PreviewTextNUM int:int, Plus or minus:bool=~) {is=title K2Node_FunctionEntry}
   > n2 = Branch(Condition="Equal (Integer)"(A=@n1.PreviewTextNUM int, B=0) {is=title K2Node_CallFunction})
   n3 = Get PreviewBox() -> (PreviewBox:Size Box Object Reference)
   > SetVisibility(self=@n3, InVisibility=Hidden)
   SetVisibility(execute=@n2.else, self=@n3, InVisibility=Visible)
   > n4 = Branch(Condition="Not Equal (Boolean)"!(A:bool=Get isCurrPlusOrMinus() -> (isCurrPlusOrMinus:bool), B:bool=@n1.Plus or minus) -> (ReturnValue:bool) {is=title K2Node_CallFunction})
   n5 = Branch(execute=[@n10.else, @n4.then], Condition=IsMobileGame())
   n6 = Branch(execute=[@n9.then, @n5.then], Condition=@n1.Plus or minus)
   > "SetText (Text)"(self=Get Symbol() -> (Symbol:Common Text Object Reference), InText="LOCTEXT(\"-\")")
   > n7 = Set isCurrPlusOrMinus(isCurrPlusOrMinus:bool=true) -> (Output_Get:bool)
   "SetText (Text)"(
    execute=[@n8.then, @n7.then]
    self=Get previewText() -> (previewText:Common Text Object Reference)
    InText="To Text (Float)"(
     Value="To Float (Integer)"(InInt=@n1.PreviewTextNUM int)
     MinimumIntegralDigits=1
     MaximumIntegralDigits=324
     MinimumFractionalDigits=0
     MaximumFractionalDigits=3
    )
   )
   > Set CurrPreviewMemory(CurrPreviewMemory:int=@n1.PreviewTextNUM int) -> (Output_Get:int=0)
   "SetText (Text)"(execute=@n6.else, self=Get Symbol(), InText="LOCTEXT(\"+\")")
   > n8 = Set isCurrPlusOrMinus(isCurrPlusOrMinus=false)
   PlayAnimation(execute=@n5.else, InAnimation=Get IN() -> (IN:Widget Animation Object Reference))
   > n9 = PlayAnimation(InAnimation=Get previewValue() -> (previewValue:Widget Animation Object Reference), NumLoopsToPlay=0, PlayMode=PingPong)
   n10 = Branch(execute=@n4.else, Condition="Equal (Integer)"(A=Get CurrPreviewMemory() -> (CurrPreviewMemory:int=0), B=@n1.PreviewTextNUM int))
   Branch(Condition=IsAnimationPlaying(InAnimation=Get Text!() -> (Text:Widget Animation Object Reference)))
   Branch(Condition=IsAnimationPlaying(InAnimation=Get Text()))
  ]}
 Textanimation {nodes [
   Textanimation() {is=title K2Node_FunctionEntry}
   > PlayAnimation(InAnimation=Get Text())
  ]}
 setTextColorWhite {nodes [
   setTextColorWhite() {is=title K2Node_FunctionEntry}
   > SetColorAndOpacity(self=[Get Symbol(), Get previewText()], InColorAndOpacity=Get whiteColor() -> (whiteColor:Slate Color Structure))
  ]}
 setTextColorred {nodes [
   setTextColorred() {is=title K2Node_FunctionEntry}
   > SetColorAndOpacity(self=[Get Symbol(), Get previewText()], InColorAndOpacity=Get redColor() -> (redColor:Slate Color Structure))
  ]}
 setWarningIconColorRed {nodes [
   setWarningIconColorRed() {is=title K2Node_FunctionEntry}
   > SetBrushTintColor!(
    execute
    self:Image Object Reference=Get WarningIcon() -> (WarningIcon:Image Object Reference)
    TintColor:Slate Color Structure=Get redColor()
   ) -> (then)
  ]}
 setWarningIconColorWhite {nodes [
   setWarningIconColorWhite() {is=title K2Node_FunctionEntry}
   > SetBrushTintColor(self=Get WarningIcon(), TintColor=Get whiteColor())
  ]}
 setWarningIconVisibility {nodes [
   n1 = setWarningIconVisibility!() -> (then, InVisibility:ESlateVisibility Enum=Visible) {is=title K2Node_FunctionEntry}
   > SetVisibility(self=Get WarningIcon(), InVisibility=@n1)
  ]}
]
variables [
 CurrPreviewMemory {type=integer}
 isCurrPlusOrMinus {type=boolean}
 redColor {type=SlateColor, default {colorUseRule=UseColor_Specified, specifiedColor=<rgba 0.849 0 0 1>}}
 White Color {type=SlateColor, default {colorUseRule=UseColor_Specified, specifiedColor=<rgba 1 0 1 1>}}
 whiteColor {type=SlateColor, default {colorUseRule=UseColor_Specified, specifiedColor=<rgba 1 1 1 1>}}
]
widgets {
 info {inheritedWidgetCount=0, parentClass=&UserWidget, rootWidgetClass=&HorizontalBox, widgetCount=9}
 widgets [
  HorizontalBox_66 {widgetClassPath=&HorizontalBox, widget=HorizontalBox_66}
  PreviewBox {
   widgetClassPath=&SizeBox
   bIsVariable=true
   parent=HorizontalBox_66
   widget=PreviewBox
   properties {bOverride_HeightOverride=true, bOverride_WidthOverride=true, heightOverride=24, visibility=Hidden, widthOverride=100}
   slot {class=HorizontalBoxSlot, properties {padding=<ltrb 5 0 0 0>}}
  }
  Overlay_1 {widgetClassPath=&Overlay, parent=PreviewBox, widget=Overlay_1, slot {class=SizeBoxSlot}}
  Image_113 {
   widgetClassPath=&Image
   bIsVariable=true
   parent=Overlay_1
   widget=Image_113
   properties {brush {
     bIsDynamicallyLoaded=false
     drawAs=Image
     imageType=NoImage
     mirroring=NoMirror
     resourceName=None
     resourceObject=null
     tiling=NoTile
     imageSize=<32 32>
     margin=<ltrb 0 0 0 0>
     outlineSettings {
      bUseBrushTransparency=false
      roundingType=HalfHeightRadius
      width=0
      color {colorUseRule=UseColor_Specified, specifiedColor=<rgba 0 0 0 0>}
      cornerRadii=<0 0 0 0>
     }
     tintColor {colorUseRule=UseColor_Specified, specifiedColor=<rgba 0 0 0 0.5>}
     uVRegion {bIsValid=false, max=<0 0>, min=<0 0>}
    }}
   slot {class=OverlaySlot, properties {horizontalAlignment=HAlign_Fill, verticalAlignment=VAlign_Fill}}
  }
  HorizontalBox_0 {
   widgetClassPath=&HorizontalBox
   parent=Overlay_1
   widget=HorizontalBox_0
   slot {class=OverlaySlot, properties {horizontalAlignment=HAlign_Fill, verticalAlignment=VAlign_Fill}}
  }
  WarningIcon {
   widgetClassPath=&Image
   bIsVariable=true
   parent=HorizontalBox_0
   widget=WarningIcon
   properties {
    bIsVolatile=true
    visibility=Collapsed
    brush {
     bIsDynamicallyLoaded=false
     drawAs=Image
     imageType=NoImage
     mirroring=NoMirror
     resourceName=None
     resourceObject=/FortniteCoreUI/HUD/MiniMap/Art/T_UI_Icon_Warning
     tiling=NoTile
     imageSize=<24 24>
     margin=<ltrb 0 0 0 0>
     outlineSettings {
      bUseBrushTransparency=false
      roundingType=HalfHeightRadius
      width=0
      color {colorUseRule=UseColor_Specified, specifiedColor=<rgba 0 0 0 0>}
      cornerRadii=<0 0 0 0>
     }
     tintColor {colorUseRule=UseColor_Specified, specifiedColor=<rgba 1 1 1 1>}
     uVRegion {bIsValid=false, max=<0 0>, min=<0 0>}
    }
   }
   slot {class=HorizontalBoxSlot, properties {horizontalAlignment=HAlign_Center, verticalAlignment=VAlign_Center}}
  }
  WrapBox_1 {
   widgetClassPath=&WrapBox
   parent=HorizontalBox_0
   widget=WrapBox_1
   properties {bExplicitWrapSize=true, wrapSize=600}
   slot {class=HorizontalBoxSlot, properties {horizontalAlignment=HAlign_Center, verticalAlignment=VAlign_Top, size {sizeRule=Fill, value=1}}}
  }
  Symbol {
   widgetClassPath=&CommonTextBlock
   bIsVariable=true
   parent=WrapBox_1
   widget=Symbol
   properties {
    text=-
    font {
     bForceMonospaced=false
     bMaterialIsStencil=false
     fontMaterial=null
     fontObject=&NotoSans
     letterSpacing=0
     monospacedWidth=1
     size=16
     skewAmount=0
     typefaceFontName=NotoSans-Bold
     outlineSettings {
      bApplyOutlineToDropShadows=false
      bMiteredCorners=false
      bSeparateFillAlpha=false
      outlineMaterial=null
      outlineSize=0
      outlineColor=<rgba 0 0 0 1>
     }
    }
    renderTransform {angle=0, scale=<1 1>, shear=<0 0>, translation=<-1 -3>}
    shadowOffset=<0 0>
   }
   slot {class=WrapBoxSlot}
  }
  previewText {
   widgetClassPath=&CommonTextBlock
   bIsVariable=true
   parent=WrapBox_1
   widget=previewText
   properties {
    text=99999
    font {
     bForceMonospaced=false
     bMaterialIsStencil=false
     fontMaterial=null
     fontObject=&NotoSans
     letterSpacing=0
     monospacedWidth=1
     size=16
     skewAmount=0
     typefaceFontName=NotoSans-Bold
     outlineSettings {
      bApplyOutlineToDropShadows=false
      bMiteredCorners=false
      bSeparateFillAlpha=false
      outlineMaterial=null
      outlineSize=0
      outlineColor=<rgba 0 0 0 1>
     }
    }
    renderTransform {angle=0, scale=<1 1>, shear=<0 0>, translation=<0 -3>}
    shadowOffset=<0 0>
   }
   slot {class=WrapBoxSlot, properties {horizontalAlignment=HAlign_Left, verticalAlignment=VAlign_Center}}
  }
 ]
}
